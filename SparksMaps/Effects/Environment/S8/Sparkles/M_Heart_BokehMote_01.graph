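# Additive unlit Niagara sprite material for sparkle bokeh motes that render as either soft circles or procedural hearts, chosen by particle random.
# The heart is built from UV math, with its scale driven by the audio-reactive BeatLifetime collection parameter.
# Camera depth fades around a focal point form a tilt-shift mask used for color blending, opacity and sprite expansion.
# Particle color, depth fade, near-camera fade and a dynamic size multiplier also shape the look.
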Material /SparksMaps/Effects/Environment/S8/Sparkles/M_Heart_BokehMote_01
graph {
 properties {
  bUsedWithNiagaraSprites=true
  blendMode=BLEND_Additive
  pixelDepthOffsetMode=PDOM_Legacy
  shadingModel=MSM_Unlit
  translucencyLightingMode=TLM_VolumetricPerVertexNonDirectional
  usedShadingModels=Unlit
 }
 nodes [
  LinearInterpolate(
   A=VectorParameter() {parameterName=ColorA, defaultValue=<rgba 1 0.006 0 1>}.RGB
   B=VectorParameter() {parameterName=ColorB, defaultValue=<rgba 0.0403 0 1 1>}.RGB
   Alpha=NamedRerouteUsage()
  )
  n1 = ParticleColor()
  MaterialFunctionCall(Emissive=@n1.RGB) {materialFunction=&MF_FX_EmissiveAdjustments}
  MaterialFunctionCall(Distance=ScalarParameter() {defaultValue=128, parameterName=NearFade}) {materialFunction=&Near_Camera_Fade}
  n2 = MaterialFunctionCall(UV Channel=TextureCoordinate()) {materialFunction=&LinearGradient}
  Multiply(A=Multiply(A=OneMinus(None=@n2.UGradient), B=Multiply(A=@n2.UGradient))) {constB=4}
  n3 = LinearInterpolate(Alpha=CollectionParameter() {collection=/FMAudioReactiveData/AudioAnalysis/MPC_Sparks_Audio_Parameters, parameterName=BeatLifetime}) {constA=0.6, constB=0.9}
  n4 = Add(
   A=ConstantBiasScale(None=MaterialFunctionCall(Texture Scale=@n3) {materialFunction=&ScaleUVsByCenter}.UVs) {bias=-0.5, scale=2}
   B=Constant2Vector() {g=0.6}
  )
  n5 = Multiply(A=Length(None=@n4) {is=class MaterialExpressionLength}) {constB=2.5}
  n6 = Multiply(A=Normalize(VectorInput=@n4)) {constB=-1}
  n7 = ComponentMask(None=@n6) {r=true}
  n8 = ComponentMask(None=@n6) {g=true}
  n9 = Multiply(A=Power(Base=Abs(None=@n7)) {constExponent=0.6667}, B=@n8)
  n10 = Subtract(A=Add(A=@n8) {constB=2}, B=ScalarParameter() {defaultValue=0.5, parameterName=Heart Taper})
  n11 = Add(A=Subtract(A=Divide(A=@n9, B=@n10), B=Multiply(A=@n8) {constB=2})) {constB=1.25}
  n12 = OneMinus(None=Saturate(None=Subtract(A=@n5, B=@n11)))
  Multiply(A=@n12)
  n13 = OneMinus(None=NamedRerouteUsage())
  Multiply(
   A=Subtract(A=WorldPosition().XYZ, B=ParticlePositionWS())
   B=Multiply(A=DynamicParameter() {paramNames [Size Multiplier, Param2, Param3, Param4]}.Size Multiplier, B=Saturate(None=@n13))
  ) {constB=0.5}
  n14 = LinearInterpolate(
   A=SphereMask(
    A=TextureCoordinate()
    B=Constant() {r=0.5}
    Radius=ScalarParameter() {defaultValue=0.35, parameterName=Radius}
    Hardness=ScalarParameter() {defaultValue=0.5, parameterName=Hardness}
   )
   B=@n12
   Alpha=Step(Y=ParticleRandom()) {constX=0.8, constY=0.8}
  )
  n15 = NamedRerouteUsage()
  Multiply(A=DepthFade(Opacity=Multiply(A=@n1.A, B=@n14)) {fadeDistanceDefault=512}, B=@n15)
  n16 = ScalarParameter() {defaultValue=4000, parameterName=FadeLength}
  n17 = ScalarParameter() {defaultValue=256, parameterName=FocalPoint}
  n18 = ScalarParameter() {defaultValue=10, parameterName=FocalPointExpand}
  n19 = MaterialFunctionCall(Fade Length=@n16, Fade Offset=Add(A=@n17, B=@n18) {constB=25}) {materialFunction=&CameraDepthFade}
  n20 = Subtract(A=@n17, B=@n18)
  n21 = Multiply(A=Multiply(
    A=OneMinus(None=@n19.Result)
    B=MaterialFunctionCall(Fade Length=@n16, Fade Offset=@n20) {materialFunction=&CameraDepthFade}.Result
   )) {constB=4}
  NamedRerouteDeclaration(None=Clamp(None=@n21)) {name=TitlShift, nodeColor=<rgba 0 1 0.8359 1>}
  OneMinus(None=@n15)
  Saturate(None=@n21)
  ScalarParameter() {defaultValue=200, parameterName=SizeScaleTest}
  TextureCoordinate()
 ]
}
